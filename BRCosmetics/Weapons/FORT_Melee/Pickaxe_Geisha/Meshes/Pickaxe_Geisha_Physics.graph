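# Physics asset for the Geisha two-handed pickaxe cosmetic, giving its dangling rope and cloth pieces simulated secondary motion.
# The handle body uses two capsules, one long along the shaft and one crosswise near the top.
# Three rope bones form a chain hanging from the handle with loose swing and tight twist limits.
# Four cloth bones each attach directly to the handle, with narrower swing on one axis and almost no twist.
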PhysicsAsset /BRCosmetics/Weapons/FORT_Melee/Pickaxe_Geisha/Meshes/Pickaxe_Geisha_Physics
bodies {
 Melee_TwoHanded_Handle [
  {shapeType=Capsule, length=96.8016, radius=2.9362, center=<1.2507 -0.6403 3.5257>}
  {shapeType=Capsule, length=23.8691, radius=10.1959, center=<-3.955 -0.5252 81.9566>, rotation=<rot 0 0 -89.9995>}
 ]
 dyn_cloth_01 [{shapeType=Capsule, length=2.0394, radius=1.882, center=<1.4486 0.5473 0>, rotation=<rot -90 0.9495 -0.9494>}]
 dyn_cloth_02 [{shapeType=Capsule, length=3.2183, radius=1.882, center=<0.9739 0.0038 0>, rotation=<rot -90 -0.0197 0.0194>}]
 dyn_cloth_03 [{shapeType=Capsule, length=2.0394, radius=1.882, center=<1.3216 0.0046 0.0032>, rotation=<rot -90 0.0011 -0.001>}]
 dyn_cloth_04 [{shapeType=Capsule, length=2.0394, radius=1.882, center=<1.197 0.5663 0>, rotation=<rot -90 -0.0197 0.0195>}]
 dyn_rope_01 [{shapeType=Capsule, length=4.1132, radius=2.6319, center=<2.8707 -0.0351 0>, rotation=<rot 90 1.8984 1.8984>}]
 dyn_rope_02 [{shapeType=Capsule, length=3.9435, radius=2.8017, center=<2.5848 0.0395 0>, rotation=<rot 90 0 0>}]
 dyn_rope_03 [{shapeType=Capsule, length=3.6924, radius=3.254, center=<3.6359 -0.0294 0>, rotation=<rot 90 0 0>}]
}
constraints [
 {bone1Name=dyn_rope_01, bone2Name=Melee_TwoHanded_Handle, swing1LimitDegrees=45, swing2LimitDegrees=45, twistLimitDegrees=5}
 {bone1Name=dyn_rope_02, bone2Name=dyn_rope_01, swing1LimitDegrees=45, swing2LimitDegrees=45, twistLimitDegrees=5}
 {bone1Name=dyn_rope_03, bone2Name=dyn_rope_02, swing1LimitDegrees=45, swing2LimitDegrees=45, twistLimitDegrees=5}
 {bone1Name=dyn_cloth_01, bone2Name=Melee_TwoHanded_Handle, swing1LimitDegrees=45, swing2LimitDegrees=20, twistLimitDegrees=1}
 {bone1Name=dyn_cloth_02, bone2Name=Melee_TwoHanded_Handle, swing1LimitDegrees=45, swing2LimitDegrees=20, twistLimitDegrees=1}
 {bone1Name=dyn_cloth_03, bone2Name=Melee_TwoHanded_Handle, swing1LimitDegrees=45, swing2LimitDegrees=20, twistLimitDegrees=1}
 {bone1Name=dyn_cloth_04, bone2Name=Melee_TwoHanded_Handle, swing1LimitDegrees=45, swing2LimitDegrees=20, twistLimitDegrees=1}
]
